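# Looping mono MetaSound preview of the Hero vehicle engine, driven by speed, throttle and boost inputs.
# Crossfaded idle/moving loops and a synth driving layer get speed-mapped pitch with a startup ramp, mixed with an ADSR-gated afterburn loop triggered by BonusSpeed.
# The mix passes through a high-pass, then a speed-swept notch with LFO-modulated bandwidth.
# IsLocal or affiliation checks gate afterburn triggering and switch notch modulation and output gain.
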MetaSoundSource /VehicleAudio/EngineAudio/Hero/MSS_EngineAudio_Hero_Preview
inputs [
 Character.Affiliation {type=Enum:FortAudioAffiliation}
 Character.Health {type=Float}
 Character.UniqueID {type=Int32}
 Fort.Vehicle.Acceleration {type=Float}
 Fort.Vehicle.BlowoffStrength {type=Float}
 Fort.Vehicle.BonusSpeed {type=Float}
 Fort.Vehicle.EnableShifting {type=Bool}
 Fort.Vehicle.Gear {type=Int32}
 Fort.Vehicle.IsGrounded {type=Bool}
 Fort.Vehicle.Load {type=Float}
 Fort.Vehicle.LowFuel {type=Bool}
 Fort.Vehicle.NormalSpeed {type=Float}
 Fort.Vehicle.OnBlowoff {type=Trigger}
 Fort.Vehicle.OnDownshift {type=Trigger}
 Fort.Vehicle.OnThrottleEngaged {type=Trigger}
 Fort.Vehicle.OnThrottleReleased {type=Trigger}
 Fort.Vehicle.OnUpshift {type=Trigger}
 Fort.Vehicle.Pitch {type=Float}
 Fort.Vehicle.RPM {type=Float}
 Fort.Vehicle.Throttle {type=Float}
 ForwardSpeed {type=Float}
 IsLocal {type=Bool}
 UE.Source.OnPlay {type=Trigger}
]
interfaces [Fort.Character:1.0, Fort.Vehicle:1.0, UE.OutputFormat.Mono:1.0, UE.Source:1.1]
outputs ["UE.OutputFormat.Mono.Audio:0" {type=Audio}]
pages [Default {
  input_defaults {
   Character.Affiliation=1
   Character.Health=0
   Character.UniqueID=0
   Fort.Vehicle.Acceleration=0
   Fort.Vehicle.BlowoffStrength=0
   Fort.Vehicle.BonusSpeed=0
   Fort.Vehicle.EnableShifting=null
   Fort.Vehicle.Gear=0
   Fort.Vehicle.IsGrounded="<MetaSoundBuilderResult.SUCCEEDED: 0>"
   Fort.Vehicle.Load=0
   Fort.Vehicle.LowFuel=null
   Fort.Vehicle.NormalSpeed=0
   Fort.Vehicle.OnBlowoff=null
   Fort.Vehicle.OnDownshift=null
   Fort.Vehicle.OnThrottleEngaged=null
   Fort.Vehicle.OnThrottleReleased=null
   Fort.Vehicle.OnUpshift=null
   Fort.Vehicle.Pitch=0
   Fort.Vehicle.RPM=0
   Fort.Vehicle.Throttle=0
   ForwardSpeed=0
   IsLocal="<MetaSoundBuilderResult.SUCCEEDED: 0>"
   UE.Source.OnPlay=null
  }
  nodes [
   InitVariable.WaveAsset:Array()
   /FortMetasound/Patches/MSP_PrintBool(Input=^IsLocal)
   &MSP_PrintBool()
   /FortMetasound/Patches/MSP_PrintInt(Input=Convert.Int32:Enum:FortAudioAffiliation("Enum:FortAudioAffiliation"=^Character.Affiliation).Int32)
   TriggerOnValueChange_0 = UE.Trigger On Value Change:Bool(Value=^IsLocal)
   TriggerOnValueChange_2 = UE.Trigger On Value Change:Int32(Value=Convert.Int32:Enum:FortAudioAffiliation("Enum:FortAudioAffiliation"=^Character.Affiliation).Int32)
   TriggerAnyTriggerAny3_0 = "TriggerAny.Trigger Any (3)"(In 0=^UE.Source.OnPlay, In 2=@TriggerOnValueChange_0.Trigger, In 1=@TriggerOnValueChange_2.Trigger)
   n1 = &TriggerCompareAffiliation(B=^Character.Affiliation, Compare=@TriggerAnyTriggerAny3_0.Out)
   n2 = TriggerCompare.Clamp:Bool(A=^IsLocal, Compare=@TriggerAnyTriggerAny3_0.Out) {also {defaults {B="<MetaSoundBuilderResult.SUCCEEDED: 0>"}}}
   TriggerAnyTriggerAny2_2 = "TriggerAny.Trigger Any (2)"(In 0=@n1.True, In 1=@n2.True)
   n3 = VariableAccessor.Trigger(Variable=VariableMutator.Trigger(Variable=InitVariable.Trigger().Variable, Value=@TriggerAnyTriggerAny2_2.Out).Variable)
   n4 = VariableAccessor.Trigger(Variable=@n3.Variable)
   n5 = VariableAccessor.Trigger(Variable=@n4.Variable)
   TriggerAccumulatorTriggerAccumulate2_0 = "TriggerAccumulator.Trigger Accumulate (2)"(In 0=@n1.False, In 1=@n2.False) {also {defaults {Auto Reset="<MetaSoundBuilderResult.SUCCEEDED: 0>"}}}
   n6 = VariableAccessor.Trigger(Variable=VariableMutator.Trigger(Variable=InitVariable.Trigger().Variable, Value=@TriggerAccumulatorTriggerAccumulate2_0.Out).Variable)
   n7 = VariableAccessor.Trigger(Variable=@n6.Variable)
   n8 = VariableAccessor.Trigger(Variable=@n7.Variable)
   TriggerControl_0 = UE.Trigger Control(
    Trigger In=UE.TriggerOnThreshold:Float(In=^Fort.Vehicle.BonusSpeed) {also {defaults {Threshold=0.9}}}.Out
    Open=VariableAccessor.Trigger(Variable=@n5.Variable).Value
    Close=VariableAccessor.Trigger(Variable=@n8.Variable).Value
   )
   ADSREnvelopeADSREnvelope_1 = ADSR Envelope.ADSR Envelope:Float(
    Trigger Attack=@TriggerControl_0.Trigger Out
    Trigger Release=UE.TriggerOnThreshold:Float(In=^Fort.Vehicle.BonusSpeed) {also {defaults {Threshold=0.9, Type=1}}}.Out
   ) {also {defaults {Attack Time=0.35, Release Time=0.75, Sustain Level=1}}}
   ADSREnvelopeADSREnvelope_0 = ADSR Envelope.ADSR Envelope:Float(Trigger Attack=^UE.Source.OnPlay) {also {defaults {Attack Time=0.75, Sustain Level=1}}}
   n9 = VariableAccessor.Float(Variable=VariableMutator.Float(
     Variable=InitVariable.Float().Variable
     Value=MapRange.MapRange:Float(In=@ADSREnvelopeADSREnvelope_0.Out Envelope) {also {defaults {Out Range A=-7, Out Range B=0}}}.Out Value
    ).Variable)
   n10 = UE.Add:Float(
    PrimaryOperand=MapRange.MapRange:Float(In=^Fort.Vehicle.NormalSpeed) {also {defaults {Out Range A=-1, Out Range B=13}}}.Out Value
    AdditionalOperands=@n9.Value
   )
   WavePlayer_1 = UE.Wave Player:Mono(Play=^UE.Source.OnPlay, Pitch Shift=@n10.Out) {also {defaults {Loop="<MetaSoundBuilderResult.SUCCEEDED: 0>", Wave Asset=&/Hero_Engine_Idle_01}, version [1,1]}}
   WavePlayer_3 = UE.Wave Player:Mono(Play=^UE.Source.OnPlay, Pitch Shift=@n10.Out) {also {defaults {Loop="<MetaSoundBuilderResult.SUCCEEDED: 0>", Wave Asset=&/Hero_Engine_MovingLoop_03}, version [1,1]}}
   AudioMixerAudioMixerMono2_1 = "AudioMixer.Audio Mixer (Mono, 2)"(
    Gain 1=^Fort.Vehicle.NormalSpeed
    Gain 0=/FortMetasound/Patches/InvertFloat(FloatIn=^Fort.Vehicle.NormalSpeed).FloatOut
    In 0=@WavePlayer_1.Out Mono
    In 1=@WavePlayer_3.Out Mono
   )
   WavePlayer_0 = UE.Wave Player:Mono(
    Play=^UE.Source.OnPlay
    Pitch Shift=UE.Add:Float(
     PrimaryOperand=MapRange.MapRange:Float(In=^Fort.Vehicle.NormalSpeed) {also {defaults {Out Range B=16.2}}}.Out Value
     AdditionalOperands=VariableAccessor.Float(Variable=@n9.Variable).Value
    ).Out
   ) {also {
     defaults {Loop="<MetaSoundBuilderResult.SUCCEEDED: 0>", Wave Asset=&/Engine_Hero_Synth_Driving_Low}
     version [1,1]
    }}
   n11 = UE.Multiply:Float(
    AdditionalOperands=/FortMetasound/Patches/MapRange_DecibelOutput(In=^Fort.Vehicle.Throttle) {also {defaults {Out Range B=2}}}.Out Value
    PrimaryOperand=UE.Multiply:Float(
     PrimaryOperand=&MapRange_DecibelOutput(In=^Fort.Vehicle.NormalSpeed) {also {defaults {In Range B=0.1, Out Range A=-6.5}}}.Out Value
     AdditionalOperands=&MapRange_DecibelOutput(In=^Fort.Vehicle.NormalSpeed) {also {defaults {Out Range B=-10}}}.Out Value
    ).Out
   )
   WavePlayer_2 = UE.Wave Player:Mono(Play=@ADSREnvelopeADSREnvelope_1.On Attack Triggered, Stop=@ADSREnvelopeADSREnvelope_1.On Done) {also {defaults {Loop="<MetaSoundBuilderResult.SUCCEEDED: 0>", Wave Asset=&/Hero_Engine_Afterburn_01}, version [1,1]}}
   AudioMixerAudioMixerMono3_0 = "AudioMixer.Audio Mixer (Mono, 3)"(
    Gain 1=UE.Multiply:Float(
     AdditionalOperands=&MapRange_DecibelOutput(In=^Fort.Vehicle.Throttle) {also {defaults {Out Range A=-6}}}.Out Value
     PrimaryOperand=&MapRange_DecibelOutput(In=^Fort.Vehicle.NormalSpeed) {also {defaults {Out Range B=-6}}}.Out Value
    ).Out
    Gain 0=UE.Multiply:Float(PrimaryOperand=UE.Reroute(Value=@ADSREnvelopeADSREnvelope_1.Out Envelope).Value) {also {defaults {AdditionalOperands=0.5623}}}.Out
    In 1=UE.Reroute(Value=@AudioMixerAudioMixerMono2_1.Out).Value
    In 2=UE.Reroute(Value=UE.Reroute(Value=@WavePlayer_0.Out Mono).Value).Value
    Gain 2=UE.Reroute(Value=UE.Multiply:Float(PrimaryOperand=@n11.Out) {also {defaults {AdditionalOperands=0.6683}}}.Out).Value
    In 0=@WavePlayer_2.Out Mono
   )
   BiquadFilter_1 = UE.Biquad Filter:Audio(In=@AudioMixerAudioMixerMono3_0.Out) {also {defaults {Bandwidth=1.6, Cutoff Frequency=90, Type=1}}}
   n12 = InitVariable.Float()
   n13 = MapRange.MapRange:Float(In=^Fort.Vehicle.Throttle) {also {defaults {Out Range A=0.5}}}
   TriggerRouteTriggerRouteFloat2_1 = "TriggerRoute.Trigger Route (Float, 2):Float"(Set 0=@n3.Value, Set 1=@n6.Value) {also {defaults {Value 0=1}}}
   n14 = UE.Multiply:Float(PrimaryOperand=^Fort.Vehicle.NormalSpeed, AdditionalOperands=@TriggerRouteTriggerRouteFloat2_1.Value)
   n15 = UE.Multiply:Float(
    AdditionalOperands=UE.LFO:Audio(Frequency=MapRange.MapRange:Float(In=^Fort.Vehicle.NormalSpeed) {also {defaults {Out Range A=0.35, Out Range B=0.5}}}.Out Value) {also {defaults {Min Value=0.8}}}.Out
    PrimaryOperand=MapRange.MapRange:Float(In=UE.Multiply:Float(AdditionalOperands=@n13.Out Value, PrimaryOperand=@n14.Out).Out) {also {defaults {Out Range A=1.5, Out Range B=3}}}.Out Value
   )
   n16 = InitVariable.Float()
   n17 = MapRange.MapRange:Float(In=^Fort.Vehicle.NormalSpeed) {also {defaults {Out Range A=550, Out Range B=4500}}}
   BiquadFilter_0 = UE.Biquad Filter:Audio(
    In=@BiquadFilter_1.Out
    Bandwidth=VariableAccessor.Float(Variable=VariableMutator.Float(Variable=@n12.Variable, Value=@n15.Out).Variable).Value
    Cutoff Frequency=VariableAccessor.Float(Variable=VariableMutator.Float(Variable=@n16.Variable, Value=@n17.Out Value).Variable).Value
   ) {also {defaults {Type=3}}}
   AudioMixerAudioMixerMono2_0 = "AudioMixer.Audio Mixer (Mono, 2)"(In 1=@BiquadFilter_0.Out)
   TriggerRouteTriggerRouteFloat2_0 = "TriggerRoute.Trigger Route (Float, 2):Float"(Set 1=@n7.Value, Set 0=@n4.Value) {also {defaults {Value 0=1, Value 1=1.1885}}}
   Multiply_0 = UE.Multiply:Audio by Float(PrimaryOperand=@AudioMixerAudioMixerMono2_0.Out, AdditionalOperands=UE.InterpTo:Audio(Target=@TriggerRouteTriggerRouteFloat2_0.Value).Value)
   TriggerOnValueChange_1 = UE.Trigger On Value Change:Bool(Value=^IsLocal)
   TriggerAnyTriggerAny2_0 = "TriggerAny.Trigger Any (2)"(In 0=^UE.Source.OnPlay, In 1=@TriggerOnValueChange_1.Trigger)
   TriggerOnValueChange_3 = UE.Trigger On Value Change:Int32(Value=Convert.Int32:Enum:FortAudioAffiliation("Enum:FortAudioAffiliation"=^Character.Affiliation).Int32)
   TriggerAnyTriggerAny2_1 = "TriggerAny.Trigger Any (2)"(In 0=^UE.Source.OnPlay, In 1=@TriggerOnValueChange_3.Trigger)
   TriggerControl_1 = UE.Trigger Control(Open=@n5.Value, Close=@n8.Value)
   TriggerRouteTriggerRouteBool2_0 = "TriggerRoute.Trigger Route (Bool, 2):Bool"(Set 1=@TriggerAccumulatorTriggerAccumulate2_0.Out, Set 0=@TriggerAnyTriggerAny2_2.Out) {also {defaults {Value 0="<MetaSoundBuilderResult.SUCCEEDED: 0>"}}}
   VariableMutator.Bool(Variable=InitVariable.Bool().Variable, Value=@TriggerRouteTriggerRouteBool2_0.Value)
  ]
  links [{dst="@UE.OutputFormat.Mono.Audio:0", src=Multiply_0.Out}]
 }]
properties {
 assetImportData=&self:AssetImportData
 attenuationSettings=&Vehicle_Engine_Att
 audioPropertiesSheet=&APS_Cosmetic_Vehicle_Engine
 bLooping=true
 duration=10000
 maxDistance=12000
 priority=80
 sampleRate=48000
 soundClassObject=&Fort_Vehicles_Engine
 soundSubmixObject=&Fortnite_SFX
 volume=2.6184
 concurrencySet [&Conc_Veh_EngineAudio_LowSpec, &Conc_Veh_EngineAudio_LowSpec_Actor]
 modulationSettings {
  highpassRouting=Union
  lowpassRouting=Union
  pitchRouting=Union
  volumeRouting=Union
  highpassModulationDestination {bEnableModulation=true, modulators [&PP_FortVehicles_Engine_HPF]}
  lowpassModulationDestination {bEnableModulation=true, modulators [&Patch_FortVehicles_Engine_LPF]}
  pitchModulationDestination {bEnableModulation=true, modulators [&DM_Bus_Vehicle_Engine_Pitch]}
  volumeModulationDestination {
   bEnableModulation=true
   modulators [&DM_VehicleEnginePatch, &Patch_FortVehicles_Engine_Vol]
  }
 }
}
